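; Quest gameplay ability, named for storm damage received, that reports a quest stat event when a configured gameplay tag is removed.
; On activation it calls the parent, then waits for RemovedTag to be removed.
; On removal it gets the quest manager from the avatar's player controller and sends a complex custom stat event carrying AdditionalTargetTags.
; No damage check appears in this graph; the trigger is purely tag-driven.
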(event Ability|EventActivateAbility
  (|CallParentFunction)
  (Ability|Tasks|WaitGameplayTagRemove (Variables|Default|GetRemovedTag)
    (:Removed
      (Quest|SendComplexCustomStatEvent (Quests|GetQuestManager (Fortnite|GetFortPlayerControllerfromActor (Ability|GetAvatarActorfromActorInfo)) "Invalid") 0 "(GameplayTags=)" (Variables|Default|GetAdditionalTargetTags) 1))))

(event Ability|EventOnEndAbility
  (params (bWasCancelled :type Boolean)))

(event Ability|EventFailedToActivatePassiveAbility
  (params (ActorInfo :type Gameplay%20Ability%20Actor%20Info%20Structure))
  (|CallParentFunction))
